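; Gameplay ability for a Season 21 battle pass quest that tracks the player starting a duel.
; On activation it calls the parent, then waits for the InDuel gameplay tag to be added.
; When the tag is added, it sends a complex custom stat event with a count of 1 to the owning player's quest manager.
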(event Ability|EventFailedToActivatePassiveAbility
  (params (ActorInfo :type Gameplay%20Ability%20Actor%20Info%20Structure)))

(event Ability|EventActivateAbility
  (|CallParentFunction)
  (Ability|Tasks|WaitGameplayTagAdd (Variables|Default|GetInDuel)
    (:Added
      (Quest|SendComplexCustomStatEvent (Variables|GAVars|Data|GetOwningPlayerQuestManager) 0 "(GameplayTags=)" "(GameplayTags=)" 1))))
